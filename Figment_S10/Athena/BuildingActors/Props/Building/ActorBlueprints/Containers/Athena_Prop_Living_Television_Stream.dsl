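; Blueprint for a living-room television streaming prop in Athena, named as a container building actor.
; BeginPlay is empty and the construction script has no logic.
; On EndPlay it destroys the actor held in its StreamingTrigger variable, so that trigger does not outlive the television.
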(event EventBeginPlay)

(event EventEndPlay
  (params (EndPlayReason :type EEndPlayReason%20Enum :default "Destroyed"))
  (Actor|DestroyActor (Variables|Default|GetStreamingTrigger)))

(fn ConstructionScript)
