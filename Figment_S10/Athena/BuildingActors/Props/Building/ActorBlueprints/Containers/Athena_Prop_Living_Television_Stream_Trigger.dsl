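; Living-room television trigger prop that starts and stops client-side streaming video as the local player's pawn moves near it.
; BeginPlay disables streaming for a command-line flag, legacy media source, cell connection or low memory.
; Inner sphere entry requests a video URL; both success and failure callbacks open the media source.
; Outer sphere exit reports time in zone and closes active players, whose close or open-failure events unbind and send streaming analytics.
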(event EventBeginPlay
  (bind _streamingsynctest (Variables|Default|GetStreamingSyncTest))
  (Networking|SwitchHasAuthority
    (:Remote
      ;[1 1 1] Check to see if television should never be turned on
      (bind _returnvalue (Utilities|GetCommandLine))
      (if (Math|Boolean|ORBoolean (Utilities|String|Contains _returnvalue "TelevisionOff") (FortGlobals|GetUseLegacyMediaSource) (Fortnite|IsCellConnection))
        (Variables|Default|SetIsActive false) ; We don't want streaming video
        (elif (>= (Fortnite|GetPhysicalMemoryMB) (FortGlobals|GetMemoryRequirementforMediaStreamingMB))
          (Variables|Default|SetIsActive true) ; The client has enough memory
          (else
            (for _array_element (Fortnite|GetLocalFortPlayerControllers true)
              (Movie|ReportVideoStreamingError _array_element _streamingsynctest false "Insufficient memory")
              (Movie|SendStreamingAnalytics _array_element _streamingsynctest)
              (Variables|Default|SetIsActive false) ; The client does not have enough memory
            )))))))

(event Collision|EventActorBeginOverlap
  (params (OtherActor :type Actor%20Object%20Reference)))

(event EventTick
  (params (DeltaSeconds :type Float%20%28single-precision%29)))

(event OnComponentBeginOverlap(Inner_Sphere_Collision)
  (params (OverlappedComponent :type Primitive%20Component%20Object%20Reference) (OtherActor :type Actor%20Object%20Reference) (OtherComp :type Primitive%20Component%20Object%20Reference) (OtherBodyIndex :type Integer) (bFromSweep :type Boolean) (SweepResult :type Hit%20Result%20Structure%20%28by%20ref%29))
  (bind _mediasource (Variables|Default|GetMediaSource))
  (Networking|SwitchHasAuthority
    (:Remote
      (bind _aspawn (Utilities|Casting|CastToPawn OtherActor))
      (if (Utilities|Equal(Object) _aspawn (Pawn|GetControlledPawn (Game|GetPlayerController 0)))
        (if (not (Variables|Default|GetEnteredStreamingZone))
          (Variables|Default|SetEnteredStreamingZone true)
          (Variables|Default|SetTimeEnteredStreamingZone (Math|DateTime|UTCNow))
          (if (Variables|Default|GetIsActive)
            (Stream|BindEventtoOnVideoUrlSuccess _mediasource (delegate Custom|On%20Video%20Url%20Success))
            (Stream|BindEventtoOnVideoUrlFailed _mediasource (delegate Custom|On%20Video%20Url%20Failed))
            (Stream|RequestVideoUrl _mediasource (Fortnite|GetFortPlayerControllerfromActor OtherActor))))))))

(event Custom|On%20Media%20Closed
  (CallFunction|TelevisionOff))

(event Custom|On%20Media%20Open%20Failed
  (params (FailedUrl :type String))
  (CallFunction|TelevisionOff))

(event Custom|On%20Video%20Url%20Success
  (params (RetrievedUrl :type String) (Code :type Integer) (RequestType :type EEpicMediaRequestType%20Enum :default "BLURL"))
  (bind _streamingsynctest (Variables|Default|GetStreamingSyncTest))
  (bind _mediasource (Variables|Default|GetMediaSource))
  (Media|MediaPlayer|OpenSource _streamingsynctest _mediasource)
  (Stream|UnbindEventfromOnVideoUrlFailed _mediasource (EventDispatchers|CreateEvent :!event Custom|On%20Video%20Url%20Failed))
  (Stream|UnbindEventfromOnVideoUrlSuccess _mediasource (EventDispatchers|CreateEvent :!event Custom|On%20Video%20Url%20Success))
  (Media|MediaPlayer|BindEventtoOnMediaOpenFailed _streamingsynctest (delegate Custom|On%20Media%20Open%20Failed))
  (Media|MediaPlayer|BindEventtoOnMediaClosed _streamingsynctest (delegate Custom|On%20Media%20Closed)))

(event Custom|On%20Video%20Url%20Failed
  (params (RetrievedUrl :type String) (Code :type Integer) (RequestType :type EEpicMediaRequestType%20Enum :default "BLURL"))
  (bind _streamingsynctest (Variables|Default|GetStreamingSyncTest))
  (bind _mediasource (Variables|Default|GetMediaSource))
  (Media|MediaPlayer|OpenSource _streamingsynctest _mediasource)
  (Stream|UnbindEventfromOnVideoUrlFailed _mediasource (EventDispatchers|CreateEvent :!event Custom|On%20Video%20Url%20Failed))
  (Stream|UnbindEventfromOnVideoUrlSuccess _mediasource (EventDispatchers|CreateEvent :!event Custom|On%20Video%20Url%20Success))
  (Media|MediaPlayer|BindEventtoOnMediaOpenFailed _streamingsynctest (delegate Custom|On%20Media%20Open%20Failed))
  (Media|MediaPlayer|BindEventtoOnMediaClosed _streamingsynctest (delegate Custom|On%20Media%20Closed)))

(event OnComponentEndOverlap(Outer_Sphere_Collision)
  (params (OverlappedComponent :type Primitive%20Component%20Object%20Reference) (OtherActor :type Actor%20Object%20Reference) (OtherComp :type Primitive%20Component%20Object%20Reference) (OtherBodyIndex :type Integer))
  (Networking|SwitchHasAuthority
    (:Remote
      (bind _aspawn (Utilities|Casting|CastToPawn OtherActor))
      (if (Utilities|Equal(Object) (Pawn|GetControlledPawn (Game|GetPlayerController 0)) _aspawn)
        (if (Variables|Default|GetEnteredStreamingZone)
          (Variables|Default|SetEnteredStreamingZone false)
          (CallFunction|SendRelevancyAnalytics)
          (if (Variables|Default|GetIsActive)
            (Media|MediaPlayer|Close (Variables|Default|GetStreamingSyncTest) :!class /Script/MediaAssets.MediaPlayer)
            (if (Game|IsActorBeingDestroyed)
              (CallFunction|TelevisionOff))))))))

(fn SendRelevancyAnalytics
  (for _array_element (Fortnite|GetLocalFortPlayerControllers true)
    (Analytics|FireEventStreamingMediaRelevancy _array_element (Math|Timespan|GetTotalSeconds (Math|DateTime|DateTime-DateTime (Math|DateTime|UTCNow) (Variables|Default|GetTimeEnteredStreamingZone))))))

(fn TelevisionOff
  (bind _streamingsynctest (Variables|Default|GetStreamingSyncTest))
  (Media|MediaPlayer|UnbindEventfromOnMediaOpenFailed _streamingsynctest (EventDispatchers|CreateEvent :!event On%20Media%20Open%20Failed))
  (Media|MediaPlayer|UnbindEventfromOnMediaClosed _streamingsynctest (EventDispatchers|CreateEvent :!event On%20Media%20Closed))
  (for _array_element (Fortnite|GetLocalFortPlayerControllers true)
    (Movie|SendStreamingAnalytics _array_element (Variables|Default|GetStreamingSyncTest))))

(fn ConstructionScript)
